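# Physics asset for the Woodsy dog pet mesh that sets up physics for its two scarf ties.
# Capsule bodies cover the left and right scarf tie bones and the upper spine bone (spine_03).
# Each tie is constrained to spine_03 with matching moderate swing limits and a tight twist limit.
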PhysicsAsset /BRCosmetics/Characters/Pets/Dog_Woodsy/Meshes/Pets_Woodsy_Physics
bodies {
 dyn_scarf_tie_l [{shapeType=Capsule, length=1.6972, radius=1.342, center=<2.3663 -0.0372 0.8208>, rotation=<rot 90 0 0>}]
 dyn_scarf_tie_r [{shapeType=Capsule, length=1.5042, radius=1.098, center=<-1.7879 0.9434 -0.5141>, rotation=<rot 90 0 0>}]
 spine_03 [{shapeType=Capsule, length=5.3705, radius=7.2112, center=<-4.6623 -0.8241 -0.5982>, rotation=<rot 90 90 -6.0613>}]
}
constraints [
 {bone1Name=dyn_scarf_tie_l, bone2Name=spine_03, swing1LimitDegrees=20, swing2LimitDegrees=20, twistLimitDegrees=5}
 {bone1Name=dyn_scarf_tie_r, bone2Name=spine_03, swing1LimitDegrees=20, swing2LimitDegrees=20, twistLimitDegrees=5}
]
