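# Post-process Anim Blueprint for the Woodsy dog pet, layering secondary motion and cosmetic fixes onto its animation.
# Optional neck and head AnimDynamics switch between speed-weighted and emote/DBNO setups, then feed a RigidBody node.
# Curves from the main AnimBP bend the paws, shrink the weapon bone and drive blink-based material hiding.
# Detecting the Pirate material alters that hiding and overrides left-eye rotation.
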AnimBlueprint /BRCosmetics/Characters/Pets/Dog_Woodsy/Meshes/Woodsy_AnimBP_PostProcess
target_skeleton=/Game/Characters/Pets/Dog_Shared/Dog_Skeleton
graphs [
 AnimGraph {nodes [
   RigidBody() {
    bOverrideWorldGravity=true
    lODThreshold=1
    simulationSpace=BaseBoneSpace
    alphaBoolBlend {blendOption=Linear}
    baseBoneRef {boneName=root}
    componentLinearAccScale=<1 1 1>
    overrideWorldGravity {z=-980}
   }
   n1 = Get PawnSpeedForAnimDynamics() -> (PawnSpeedForAnimDynamics:float)
   n2 = Anim Dynamics - Bone: neck_02(ComponentPose=Local To Component(LocalPose=Use cached pose 'Sub Graph Cache'()), Alpha=@n1, ExternalForce="0,0,0") {
    angularDampingOverride=0.999
    angularSpringConstant=200
    bGravityOverrideInSimSpace=true
    bOverrideAngularDamping=true
    gravityScale=0.4
    lODThreshold=1
    numSolverIterationsPostUpdate=4
    numSolverIterationsPreUpdate=16
    alphaBoolBlend {blendOption=Linear}
    alphaScaleBiasClamp {bInterpResult=true, bMapRange=true, inRange {max=500}}
    boundBone {boneName=neck_02}
    componentLinearAccScale=<0.5 0.5 0.5>
    physicsBodyDefinitions [{
      sphereCollisionRadius=0
      boundBone {boneName=neck_02}
      boxExtents=<3 10 10>
      constraintSetup {angularTargetAxis=AxisY, twistAxis=AxisY, angularLimitsMax=<10 3 3>, angularLimitsMin=<-10 -3 -3>, angularTarget=<0 0 0>}
      localJointOffset=<0 1 0>
     }]
   }
   n3 = Anim Dynamics - Bone: head(ComponentPose=@n2, Alpha=@n1, ExternalForce="0,0,0") {
    angularDampingOverride=0.999
    angularSpringConstant=200
    bGravityOverrideInSimSpace=true
    bOverrideAngularDamping=true
    gravityScale=0.4
    lODThreshold=1
    numSolverIterationsPostUpdate=4
    numSolverIterationsPreUpdate=16
    alphaBoolBlend {blendOption=Linear}
    alphaScaleBiasClamp {bInterpResult=true, bMapRange=true, inRange {max=500}}
    boundBone {boneName=head}
    componentLinearAccScale=<0.5 0.5 0.5>
    physicsBodyDefinitions [{
      sphereCollisionRadius=0
      boundBone {boneName=head}
      boxExtents=<3 10 10>
      constraintSetup {angularTargetAxis=AxisY, twistAxis=AxisY, angularLimitsMax=<10 3 3>, angularLimitsMin=<-10 -3 -3>, angularTarget=<0 0 0>}
      localJointOffset=<0 1 0>
     }]
   }
   n4 = Anim Dynamics - Bone: neck_02(ComponentPose=Local To Component(LocalPose=Use cached pose 'Sub Graph Cache'()), Alpha=1, ExternalForce="0,0,0") {
    angularDampingOverride=0.99
    angularSpringConstant=1500
    bAngularSpring=true
    bGravityOverrideInSimSpace=true
    bOverrideAngularDamping=true
    gravityScale=0.3
    lODThreshold=1
    numSolverIterationsPostUpdate=2
    numSolverIterationsPreUpdate=8
    simulationSpace=RootRelative
    alphaBoolBlend {blendOption=Linear}
    boundBone {boneName=neck_02}
    componentLinearAccScale=<0.5 0.5 0.5>
    componentLinearVelScale=<0.5 0.5 0.5>
    physicsBodyDefinitions [{
      sphereCollisionRadius=0
      boundBone {boneName=neck_02}
      boxExtents=<5 5 5>
      constraintSetup {angularTargetAxis=AxisY, twistAxis=AxisY, angularLimitsMax=<15 10 2>, angularLimitsMin=<-15 -10 -20>, angularTarget=<0 0 0>}
      localJointOffset=<0 0.25 0>
     }]
    relativeSpaceBone {boneName=root}
   }
   n5 = Anim Dynamics - Bone: head(ComponentPose=@n4, Alpha=1, ExternalForce="0,0,0") {
    angularDampingOverride=0.99
    angularSpringConstant=1500
    bAngularSpring=true
    bGravityOverrideInSimSpace=true
    bOverrideAngularDamping=true
    gravityScale=0.3
    lODThreshold=1
    numSolverIterationsPostUpdate=2
    numSolverIterationsPreUpdate=8
    simulationSpace=RootRelative
    alphaBoolBlend {blendOption=Linear}
    boundBone {boneName=head}
    componentLinearAccScale=<0.5 0.5 0.5>
    componentLinearVelScale=<0.5 0.5 0.5>
    physicsBodyDefinitions [{
      sphereCollisionRadius=0
      boundBone {boneName=head}
      boxExtents=<5 5 5>
      constraintSetup {angularTargetAxis=AxisY, twistAxis=AxisY, angularLimitsMax=<15 10 2>, angularLimitsMin=<-15 -10 -20>, angularTarget=<0 0 0>}
      localJointOffset=<0 0.25 0>
     }]
    relativeSpaceBone {boneName=root}
   }
   n6 = Blend Poses by bool(
    BlendPose_0=Component To Local(ComponentPose=@n5)
    BlendPose_1=Component To Local(ComponentPose=@n3)
    BlendTime_0=1
    BlendTime_1=1
    bActiveValue=Get bIsPlayingEmoteOrDBNO() -> (bIsPlayingEmoteOrDBNO:bool)
   ) {blendType=Linear}
   n7 = Use cached pose 'Sub Graph Cache'()
   n8 = Get bPetWantsAnimDynamics() -> (bPetWantsAnimDynamics:bool)
   n9 = RigidBody(ComponentPose=Local To Component(LocalPose=Blend Poses by bool(BlendPose_0=@n6, BlendPose_1=@n7, BlendTime_0=1, BlendTime_1=1, bActiveValue=@n8) {blendType=Linear})) {
    bOverrideWorldGravity=true
    overridePhysicsAsset=&/Pets_Woodsy_Physics
    alphaBoolBlend {blendOption=Linear}
    componentAppliedLinearAccClamp=<100 100 100>
    componentLinearAccScale=<1 1 1>
    overrideWorldGravity {z=-980}
   }
   n10 = Get isInIdle() -> (isInIdle:double)
   n11 = "Transform (Modify) Bone - Bone: hand_r"(ComponentPose=@n9, Alpha=@n10, Translation="0,0,0", Rotation="-10,0,0", Scale="1,1,1") {
    rotationMode=BMM_Additive
    alphaBoolBlend {blendInTime=0.1, blendOption=Linear, blendOutTime=0.1}
    boneToModify {boneName=hand_r}
    rotation {pitch=-10}
   }
   n12 = "Transform (Modify) Bone - Bone: hand_l"(ComponentPose=@n11, Alpha=@n10, Translation="0,0,0", Rotation="-50,0,0", Scale="1,1,1") {
    rotationMode=BMM_Additive
    alphaBoolBlend {blendInTime=0.1, blendOption=Linear, blendOutTime=0.1}
    boneToModify {boneName=hand_l}
    rotation {pitch=-50}
   }
   n13 = "Transform (Modify) Bone - Bone: L_eye"!(ComponentPose=@n12, bAlphaBoolEnabled:bool=Get pirateActive() -> (pirateActive:bool), Translation="0,0,0", Rotation="0,180,0", Scale="1,1,1") -> (Pose) {
    alphaInputType=Bool
    rotationMode=BMM_Replace
    rotationSpace=BCS_ParentBoneSpace
    alphaBoolBlend {blendInTime=0.1, blendOption=Linear, blendOutTime=0.1}
    boneToModify {boneName=L_eye}
    rotation {yaw=180}
   }
   n14 = "Transform (Modify) Bone - Bone: weapon"(ComponentPose=@n13, Alpha=Get hasBone() -> (hasBone:double), Translation="0,0,0", Rotation="0,180,0", Scale="0.01,0.01,0.01") {
    rotationSpace=BCS_ParentBoneSpace
    scaleMode=BMM_Replace
    alphaBoolBlend {blendInTime=0.1, blendOption=Linear, blendOutTime=0.1}
    alphaScaleBiasClamp {bMapRange=true, outRange {max=0, min=1}}
    boneToModify {boneName=weapon}
    rotation {yaw=180}
    scale=<0.01 0.01 0.01>
   }
   Output Pose(Result=Component To Local(ComponentPose=@n14)) {layerGroup=None}
   AnimGraphNode_SubInput_1 = Input Pose()
   Save cached pose 'Sub Graph Cache'(Pose=@AnimGraphNode_SubInput_1) {cacheName=Sub Graph Cache}
  ]}
 EventGraph {nodes [
   Event BlueprintUpdateAnimation()
   > n1 = Cast To FortPlayerPet!(execute, Object=GetOwningActor()) -> (then, CastFailed, AsFort Player Pet:Fort Player Pet Object Reference) {is=title K2Node_DynamicCast}
   > n2 = Cast To B_AnimBP_Woodsy!(
    execute
    Object=GetAnimInstance(self=Get PetMesh!(self:Fort Player Pet Object Reference=@n1) -> (PetMesh:Skeletal Mesh Component Budgeted Object Reference))
   ) -> (then, CastFailed, AsB Anim BP Woodsy:B Anim BP Woodsy Object Reference) {is=title K2Node_DynamicCast}
   > n3 = Get Dynamic Material Instance(SkeletalMeshComponent=GetOwningComponent())
   > Set isInIdle(isInIdle:double=GetCurveValue(self=@n2, CurveName=WoodsyPaws)) -> (Output_Get:double)
   > Set hasBone(hasBone:double=GetCurveValue(self=@n2, CurveName=hasBone)) -> (Output_Get:double)
   > n4 = Branch(Condition="Equal (Object)"(
     A=Get Parent!(self:Material Instance Object Reference=@n3) -> (Parent:Material Interface Object Reference)
     B=/BRCosmetics/Characters/Pets/Dog_Woodsy/Skin/Woodsy_Pirate/Materials/MI_Pet_Woodsy_Pirate
    ) {is=title K2Node_CallFunction})
   n5 = MapRangeClamped(Value=GetCurveValue(self=@n2, CurveName=R_blink_pose), InRangeA=0, InRangeB=0.1, OutRangeA=1, OutRangeB=0)
   > SetScalarParameterValue(self=@n3, ParameterName=Hide Element 01, Value=@n5)
   > SetScalarParameterValue(self=@n3, ParameterName=Hide Element 03, Value=0)
   > Set pirateActive(pirateActive:bool=true) -> (Output_Get:bool)
   SetScalarParameterValue(execute=@n4.else, self=@n3, ParameterName=Hide Element 01, Value=@n5)
   > SetScalarParameterValue(self=@n3, ParameterName=Hide Element 03, Value=@n5)
   > Set pirateActive(pirateActive=false)
   Event BlueprintInitializeAnimation()
  ]}
]
variables [
 hasBone {type=number}
 isInIdle {type=number}
 pirateActive {type=boolean}
]
